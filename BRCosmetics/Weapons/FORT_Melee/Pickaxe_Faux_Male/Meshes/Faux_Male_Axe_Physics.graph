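# Physics asset for the Faux Male two-handed pickaxe, giving collision bodies to the weapon handle and a small charm joint chain.
# A large capsule covers the handle, and small capsules sit on the charm's master joint and its two child joints.
# Constraints chain handle to master to joint 1 to joint 2.
# The master rotates freely on the handle, while the first child has tighter angular limits than the second.
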PhysicsAsset /BRCosmetics/Weapons/FORT_Melee/Pickaxe_Faux_Male/Meshes/Faux_Male_Axe_Physics
bodies {
 Melee_TwoHanded_Handle [{shapeType=Capsule, length=38.665, radius=5.7145, center=<-0.1077 -2.3234 18.3817>, rotation=<rot -0.8854 -90 -0.0067>}]
 dyn_charm_jnt_1 [{shapeType=Capsule, length=0.5, radius=1.5, center=<0.466 0 0>, rotation=<rot 90 0 0>}]
 dyn_charm_jnt_2 [{shapeType=Capsule, length=0.5, radius=1.5, center=<0.466 0 0>, rotation=<rot 90 0 0>}]
 dyn_charm_jnt_master [{shapeType=Capsule, length=1, radius=1, rotation=<rot 90 0 0>}]
}
constraints [
 {
  bone1Name=dyn_charm_jnt_master
  bone2Name=Melee_TwoHanded_Handle
  swing1LimitDegrees=45
  swing1Motion=Free
  swing2LimitDegrees=45
  swing2Motion=Free
  twistLimitDegrees=45
  twistMotion=Free
 }
 {bone1Name=dyn_charm_jnt_1, bone2Name=dyn_charm_jnt_master, swing1LimitDegrees=45, swing2LimitDegrees=25, twistLimitDegrees=5}
 {bone1Name=dyn_charm_jnt_2, bone2Name=dyn_charm_jnt_1, swing1LimitDegrees=70, swing2LimitDegrees=70, twistLimitDegrees=45}
]
